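# Physics asset for the ears of the M_MED_WidePlane backpack cosmetic, defining collision bodies and joints for the left and right ear bones.
# A capsule on pack_head anchors two mirrored two-bone chains, with each ear_01 bone constrained to the head and each ear_02 bone constrained to its ear_01.
# All four joints share tight swing limits and an even tighter twist limit, allowing only slight ear movement.
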PhysicsAsset /BRCosmetics/Accessories/FORT_Backpacks/Backpack_M_MED_WidePlane/Meshes/M_MED_WidePlane_Pack_Ears_Physics
bodies {
 dyn_ear_01_l [{shapeType=Capsule, length=3.3397, radius=2.1433, shapeName=dyn_ear_01_l_capsule, center=<2.1451 0 0>, rotation=<rot 90 0 0>}]
 dyn_ear_01_r [{shapeType=Capsule, length=3.3397, radius=2.1433, shapeName=dyn_ear_01_r_capsule, center=<-1.655 0 0>, rotation=<rot 90 0 0>}]
 dyn_ear_02_l [{shapeType=Capsule, length=3.6072, radius=1.8758, shapeName=dyn_ear_02_l_capsule, center=<1.332 0 0>, rotation=<rot 90 0 0>}]
 dyn_ear_02_r [{shapeType=Capsule, length=3.6072, radius=1.8758, shapeName=dyn_ear_02_r_capsule, center=<-1.222 0 0>, rotation=<rot 90 0 0>}]
 pack_head [{shapeType=Capsule, length=2.2173, radius=6.7827, shapeName=pack_head_capsule, rotation=<rot 90 0 0>}]
}
constraints [
 {bone1Name=dyn_ear_02_l, bone2Name=dyn_ear_01_l, swing1LimitDegrees=8, swing2LimitDegrees=8, twistLimitDegrees=2}
 {bone1Name=dyn_ear_02_r, bone2Name=dyn_ear_01_r, swing1LimitDegrees=8, swing2LimitDegrees=8, twistLimitDegrees=2}
 {bone1Name=dyn_ear_01_l, bone2Name=pack_head, swing1LimitDegrees=8, swing2LimitDegrees=8, twistLimitDegrees=2}
 {bone1Name=dyn_ear_01_r, bone2Name=pack_head, swing1LimitDegrees=8, swing2LimitDegrees=8, twistLimitDegrees=2}
]
