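# Physics asset for the M_MED_WidePlane backpack cosmetic, whose skeleton carries an upper-body figure with head, arms and sleeves.
# All bodies are capsules: a large pack body constrained to root supports a spine, neck and head chain.
# Both arms branch from the upper spine and end in three-link sleeve chains, the left via a wrist body.
# Most joints share wide swing and tight twist limits, while a bag-edge body rotates freely on the pack.
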PhysicsAsset /BRCosmetics/Accessories/FORT_Backpacks/Backpack_M_MED_WidePlane/Meshes/M_MED_WidePlane_Pack_Physics
bodies {
 bag_edge_03 [{
   shapeType=Capsule
   length=2.9418
   radius=0.4467
   shapeName=bag_edge_03_capsule
   center=<1.1657 -0.2436 -0.9258>
   rotation=<rot -2.8371 -90 -2.6117>
  }]
 dyn_lowerarm_r [{shapeType=Capsule, length=3.7251, radius=1.628, shapeName=dyn_lowerarm_r_capsule, center=<1.6604 0 0>, rotation=<rot 90 0 0>}]
 dyn_sleeve_01_l [{shapeType=Capsule, length=1.6957, radius=1.38, shapeName=dyn_sleeve_01_l_capsule, center=<2.0235 0 0>, rotation=<rot 90 0 0>}]
 dyn_sleeve_02_l [{shapeType=Capsule, length=1.6957, radius=1.38, shapeName=dyn_sleeve_02_l_capsule, center=<2.0235 0 0>, rotation=<rot 90 0 0>}]
 dyn_sleeve_03_l [{shapeType=Capsule, length=1.6957, radius=1.38, shapeName=dyn_sleeve_03_l_capsule, center=<2.0235 0 0>, rotation=<rot 90 0 0>}]
 dyn_upperarm_r [{shapeType=Capsule, length=4.2781, radius=2, shapeName=dyn_upperarm_r_capsule, center=<1.4747 0 0>, rotation=<rot 90 0 0>}]
 pack [{shapeType=Capsule, length=7.2323, radius=10.8697, shapeName=pack_capsule, center=<0 -8.5665 -13.8831>, rotation=<rot 0 0 -8.8405>}]
 pack_dyn_sleeve_01_r [{shapeType=Capsule, length=1.6957, radius=1.38, shapeName=dyn_sleeve_01_l_capsule, center=<2.0235 0 0>, rotation=<rot 90 0 0>}]
 pack_dyn_sleeve_02_r [{shapeType=Capsule, length=1.6957, radius=1.38, shapeName=dyn_sleeve_02_l_capsule, center=<2.0235 0 0>, rotation=<rot 90 0 0>}]
 pack_dyn_sleeve_03_r [{shapeType=Capsule, length=1.6957, radius=1.38, shapeName=dyn_sleeve_03_l_capsule, center=<2.0235 0 0>, rotation=<rot 90 0 0>}]
 pack_head [{shapeType=Capsule, length=2.2173, radius=6.7827, shapeName=head_capsule, rotation=<rot 90 0 0>}]
 pack_lowerarm_l [{shapeType=Capsule, length=0.505, radius=0.505, shapeName=lowerarm_l_capsule, rotation=<rot 0 -90 0>}]
 pack_neck_01 [{shapeType=Capsule, length=4.7878, radius=4.2122, shapeName=neck_01_capsule}]
 pack_neck_02 [{shapeType=Capsule, length=4.7878, radius=4.2122, shapeName=neck_02_capsule}]
 pack_spine_01 [{shapeType=Capsule, length=5.3071, radius=6.1355, shapeName=spine_01_capsule}]
 pack_spine_02 [{shapeType=Capsule, length=8.7469, radius=4.2254, shapeName=spine_02_capsule}]
 pack_spine_03 [{shapeType=Capsule, length=6.1471, radius=4, shapeName=spine_03_capsule}]
 pack_upperarm_l [{
   shapeType=Capsule
   length=7.4168
   radius=2.9433
   shapeName=upperarm_l_capsule
   center=<4.5609 -0.4239 0.2334>
   rotation=<rot 0 -74.1676 -89.6726>
  }]
 pack_wrist_l [{shapeType=Capsule, length=0.505, radius=0.505, shapeName=wrist_l_capsule, rotation=<rot 0 -90 0>}]
 root [{shapeType=Capsule, length=0.505, radius=0.505, shapeName=root_capsule, rotation=<rot 0 -90 0>}]
}
constraints [
 {bone1Name=dyn_lowerarm_r, bone2Name=dyn_upperarm_r, swing1LimitDegrees=45, swing2LimitDegrees=45, twistLimitDegrees=5}
 {bone1Name=pack, bone2Name=root, swing1LimitDegrees=20, swing2LimitDegrees=0, twistLimitDegrees=0}
 {
  bone1Name=bag_edge_03
  bone2Name=pack
  swing1LimitDegrees=45
  swing1Motion=Free
  swing2LimitDegrees=45
  swing2Motion=Free
  twistLimitDegrees=45
  twistMotion=Free
 }
 {bone1Name=dyn_sleeve_02_l, bone2Name=dyn_sleeve_01_l, swing1LimitDegrees=45, swing2LimitDegrees=45, twistLimitDegrees=5}
 {bone1Name=dyn_sleeve_03_l, bone2Name=dyn_sleeve_02_l, swing1LimitDegrees=45, swing2LimitDegrees=45, twistLimitDegrees=5}
 {bone1Name=pack_dyn_sleeve_01_r, bone2Name=dyn_lowerarm_r, swing1LimitDegrees=45, swing2LimitDegrees=45, twistLimitDegrees=5}
 {bone1Name=pack_dyn_sleeve_02_r, bone2Name=pack_dyn_sleeve_01_r, swing1LimitDegrees=45, swing2LimitDegrees=45, twistLimitDegrees=5}
 {bone1Name=pack_dyn_sleeve_03_r, bone2Name=pack_dyn_sleeve_02_r, swing1LimitDegrees=45, swing2LimitDegrees=45, twistLimitDegrees=5}
 {bone1Name=dyn_sleeve_01_l, bone2Name=pack_wrist_l, swing1LimitDegrees=45, swing2LimitDegrees=45, twistLimitDegrees=5}
 {bone1Name=dyn_upperarm_r, bone2Name=pack_spine_03, swing1LimitDegrees=45, swing2LimitDegrees=45, twistLimitDegrees=5}
 {bone1Name=pack_spine_01, bone2Name=pack, swing1LimitDegrees=45, swing2LimitDegrees=45, twistLimitDegrees=5}
 {bone1Name=pack_spine_02, bone2Name=pack_spine_01, swing1LimitDegrees=45, swing2LimitDegrees=45, twistLimitDegrees=5}
 {bone1Name=pack_spine_03, bone2Name=pack_spine_02, swing1LimitDegrees=45, swing2LimitDegrees=45, twistLimitDegrees=5}
 {bone1Name=pack_neck_01, bone2Name=pack_spine_03, swing1LimitDegrees=45, swing2LimitDegrees=45, twistLimitDegrees=5}
 {bone1Name=pack_neck_02, bone2Name=pack_neck_01, swing1LimitDegrees=45, swing2LimitDegrees=45, twistLimitDegrees=5}
 {bone1Name=pack_head, bone2Name=pack_neck_02, swing1LimitDegrees=45, swing2LimitDegrees=45, twistLimitDegrees=5}
 {bone1Name=pack_upperarm_l, bone2Name=pack_spine_03, swing1LimitDegrees=45, swing2LimitDegrees=45, twistLimitDegrees=5}
 {bone1Name=pack_lowerarm_l, bone2Name=pack_upperarm_l, swing1LimitDegrees=45, swing2LimitDegrees=45, twistLimitDegrees=5}
 {bone1Name=pack_wrist_l, bone2Name=pack_lowerarm_l, swing1LimitDegrees=45, swing2LimitDegrees=45, twistLimitDegrees=5}
]
