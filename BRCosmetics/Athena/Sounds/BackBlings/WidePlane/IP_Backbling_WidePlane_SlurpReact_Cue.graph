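# Sound cue for the WidePlane back bling's reaction sound, named as its slurp react.
# A Random node picks one of five equally weighted react waves.
# A Modulator adds slight pitch and volume variation on each play.
# It plays at reduced volume through the back bling sound class with close-range pet attenuation.
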SoundCue /BRCosmetics/Athena/Sounds/BackBlings/WidePlane/IP_Backbling_WidePlane_SlurpReact_Cue
root=Modulator_1
nodes [
 {id=Modulator_1, children [Random_0], properties {pitchMax=1.02, pitchMin=0.98, volumeMax=1.02, volumeMin=0.98}}
 {id=Random_0, children [WavePlayer_0,WavePlayer_1,WavePlayer_2,WavePlayer_3,WavePlayer_4], properties {weights [1,1,1,1,1]}}
 {id=WavePlayer_0, properties {soundWaveAssetPtr=&/IP_Backpack_WidePlane_React_01}}
 {id=WavePlayer_1, properties {soundWaveAssetPtr=&/IP_Backpack_WidePlane_React_02}}
 {id=WavePlayer_2, properties {soundWaveAssetPtr=&/IP_Backpack_WidePlane_React_03}}
 {id=WavePlayer_3, properties {soundWaveAssetPtr=&/IP_Backpack_WidePlane_React_04}}
 {id=WavePlayer_4, properties {soundWaveAssetPtr=&/IP_Backpack_WidePlane_React_05}}
]
properties {
 attenuationSettings=&Fort_Pets_Close_Athena_Att
 maxDistance=1800.0001
 soundClassObject=/Game/Sounds/FortSoundClassesAndModes/Fort_Backblings
 volumeMultiplier=0.4
}
