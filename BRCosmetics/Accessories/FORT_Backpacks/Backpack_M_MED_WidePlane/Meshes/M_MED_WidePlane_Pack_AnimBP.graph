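# Animation Blueprint for the WidePlane back bling, animating a pack character with idle and reactive states.
# The output chain runs the Idle and Reactive state machine through bone offsets (some skydiving-only), RigidBody, then AnimDynamics springs on hair and pack.
# A separate Cane/NoCane machine nests idle, meditate, chest-open and Chug Jug reactions, but is unconnected to the output.
# A RigidBodyWithControl setup for head, arms, sleeves and spine also sits unwired.
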AnimBlueprint /BRCosmetics/Accessories/FORT_Backpacks/Backpack_M_MED_WidePlane/Meshes/M_MED_WidePlane_Pack_AnimBP
target_skeleton=&/M_MED_WidePlane_Pack_Skeleton
graphs [
 AnimGraph {nodes [
   n1 = Get bIsSkyDiving()
   "Transform (Modify) Bone - Bone: hand_r"!(ComponentPose, bAlphaBoolEnabled:bool=@n1, Translation="0,2,0", Rotation="0,0,20", Scale="1,1,1") -> (Pose) {
    alphaInputType=Bool
    rotationMode=BMM_Additive
    rotationSpace=BCS_BoneSpace
    translationMode=BMM_Additive
    translationSpace=BCS_BoneSpace
    alphaBoolBlend {blendOption=Linear}
    boneToModify {boneName=hand_r}
    rotation {roll=20}
    translation {y=2}
   }
   RigidBodyWithControl!(
    ComponentPose:Component Space Pose Link Structure
    Alpha:float=Get PBA_Alpha() -> (PBA_Alpha:double)
    OverrideWorldGravity:Vector="0,0,0"
    ExternalForce:Vector="0,0,0"
    ForcesAndTorques:Rigid Body Forces And Torques Structure=(ForcesAndTorques=())
   ) -> (Pose:Component Space Pose Link Structure) {
    is=title AnimGraphNode_RigidBodyWithControl
    bEnableControls=true
    bOverrideWorldGravity=true
    evaluationResetTime=0.2
    overridePhysicsAsset=&/M_MED_WidePlane_Pack_Physics
    physicsAssetAuthoredSkeletalMesh=&/M_MED_WidePlane_Pack
    additionalSets {
     controlSetUpdates [
      {setName=HeadControlSet, names [Head]}
      {setName=Arm_L_ControlSet, names [Arm_L]}
      {setName=Arm_R_ControlSet, names [Arm_R]}
      {setName=Sleeve_R_ControlSet, names [Sleeve_R]}
      {setName=Sleeve_L_ControlSet, names [Sleeve_L]}
      {setName=SpineControlSet, names [Spine]}
     ]
     modifierSetUpdates [
      {setName=HeadSet, names [Head]}
      {setName=Arm_L_Set, names [Arm_L]}
      {setName=Arm_R_Set, names [Arm_R]}
      {setName=Sleeve_R_Set, names [Sleeve_R]}
      {setName=Sleeve_L_Set, names [Sleeve_L]}
      {setName=SpineSet, names [Spine]}
     ]
    }
    alphaBoolBlend {blendOption=Linear}
    characterSetupData {
     defaultBodyModifierData {movementType=Kinematic}
     defaultParentSpaceControlData {bEnabled=false}
     defaultWorldSpaceControlData {bEnabled=false}
     limbSetupData [
      {
       bCreateBodyModifiers=true
       bCreateParentSpaceControls=true
       bCreateWorldSpaceControls=true
       bIncludeParentBone=false
       limbName=Sleeve_R
       startBone=pack_dyn_sleeve_01_r
      }
      {
       bCreateBodyModifiers=true
       bCreateParentSpaceControls=true
       bCreateWorldSpaceControls=true
       bIncludeParentBone=false
       limbName=Sleeve_L
       startBone=dyn_sleeve_01_l
      }
      {
       bCreateBodyModifiers=true
       bCreateParentSpaceControls=true
       bCreateWorldSpaceControls=true
       bIncludeParentBone=false
       limbName=Head
       startBone=pack_neck_01
      }
      {
       bCreateBodyModifiers=true
       bCreateParentSpaceControls=true
       bCreateWorldSpaceControls=true
       bIncludeParentBone=false
       limbName=Wrist_L
       startBone=pack_wrist_l
      }
      {
       bCreateBodyModifiers=true
       bCreateParentSpaceControls=true
       bCreateWorldSpaceControls=true
       bIncludeParentBone=false
       limbName=Arm_R
       startBone=dyn_upperarm_r
      }
      {
       bCreateBodyModifiers=true
       bCreateParentSpaceControls=true
       bCreateWorldSpaceControls=true
       bIncludeParentBone=false
       limbName=Arm_L
       startBone=pack_upperarm_l
      }
      {
       bCreateBodyModifiers=true
       bCreateParentSpaceControls=true
       bCreateWorldSpaceControls=true
       bIncludeParentBone=false
       limbName=Spine
       startBone=pack_spine_01
      }
     ]
    }
    controlAndModifierParameters {
     controlParameters [
      HeadControlSet {data {
        angularDampingRatio=0.7
        angularExtraDamping=0
        angularStrength=2.5
        angularTargetVelocityMultiplier=1
        bDisableCollision=false
        bEnableAngularDampingRatio=true
        bEnableAngularExtraDamping=true
        bEnableAngularStrength=true
        bEnableAngularTargetVelocityMultiplier=true
        bEnableCustomControlPoint=true
        bEnableLinearDampingRatio=true
        bEnableLinearExtraDamping=true
        bEnableLinearStrength=true
        bEnableLinearTargetVelocityMultiplier=true
        bEnableMaxForce=true
        bEnableMaxTorque=true
        bEnablebDisableCollision=true
        bEnablebEnabled=true
        bEnablebOnlyControlChildObject=true
        bEnablebUseAccelerationDriveMode=true
        bEnablebUseCustomControlPoint=true
        bEnablebUseSkeletalAnimation=true
        bEnabled=true
        bOnlyControlChildObject=false
        bUseAccelerationDriveMode=true
        bUseCustomControlPoint=false
        bUseSkeletalAnimation=true
        linearDampingRatio=0.7
        linearExtraDamping=0
        linearStrength=2.5
        linearTargetVelocityMultiplier=1
        maxForce=0
        maxTorque=0
        customControlPoint=<0 0 0>
       }}
      Arm_L_ControlSet {data {
        angularDampingRatio=0.7
        angularExtraDamping=0
        angularStrength=3
        angularTargetVelocityMultiplier=1
        bDisableCollision=false
        bEnableAngularDampingRatio=true
        bEnableAngularExtraDamping=true
        bEnableAngularStrength=true
        bEnableAngularTargetVelocityMultiplier=true
        bEnableCustomControlPoint=true
        bEnableLinearDampingRatio=true
        bEnableLinearExtraDamping=true
        bEnableLinearStrength=true
        bEnableLinearTargetVelocityMultiplier=true
        bEnableMaxForce=true
        bEnableMaxTorque=true
        bEnablebDisableCollision=true
        bEnablebEnabled=true
        bEnablebOnlyControlChildObject=true
        bEnablebUseAccelerationDriveMode=true
        bEnablebUseCustomControlPoint=true
        bEnablebUseSkeletalAnimation=true
        bEnabled=true
        bOnlyControlChildObject=false
        bUseAccelerationDriveMode=true
        bUseCustomControlPoint=false
        bUseSkeletalAnimation=true
        linearDampingRatio=0.7
        linearExtraDamping=0
        linearStrength=3
        linearTargetVelocityMultiplier=1
        maxForce=0
        maxTorque=0
        customControlPoint=<0 0 0>
       }}
      Arm_R_ControlSet {data {
        angularDampingRatio=0.7
        angularExtraDamping=0
        angularStrength=5
        angularTargetVelocityMultiplier=1
        bDisableCollision=false
        bEnableAngularDampingRatio=true
        bEnableAngularExtraDamping=true
        bEnableAngularStrength=true
        bEnableAngularTargetVelocityMultiplier=true
        bEnableCustomControlPoint=true
        bEnableLinearDampingRatio=true
        bEnableLinearExtraDamping=true
        bEnableLinearStrength=true
        bEnableLinearTargetVelocityMultiplier=true
        bEnableMaxForce=true
        bEnableMaxTorque=true
        bEnablebDisableCollision=true
        bEnablebEnabled=true
        bEnablebOnlyControlChildObject=true
        bEnablebUseAccelerationDriveMode=true
        bEnablebUseCustomControlPoint=true
        bEnablebUseSkeletalAnimation=true
        bEnabled=true
        bOnlyControlChildObject=true
        bUseAccelerationDriveMode=true
        bUseCustomControlPoint=false
        bUseSkeletalAnimation=true
        linearDampingRatio=0.7
        linearExtraDamping=0
        linearStrength=5
        linearTargetVelocityMultiplier=1
        maxForce=0
        maxTorque=0
        customControlPoint=<0 0 0>
       }}
      Sleeve_R_ControlSet {data {
        angularDampingRatio=0.7
        angularExtraDamping=0
        angularStrength=4
        angularTargetVelocityMultiplier=1
        bDisableCollision=false
        bEnableAngularDampingRatio=true
        bEnableAngularExtraDamping=true
        bEnableAngularStrength=true
        bEnableAngularTargetVelocityMultiplier=true
        bEnableCustomControlPoint=true
        bEnableLinearDampingRatio=true
        bEnableLinearExtraDamping=true
        bEnableLinearStrength=true
        bEnableLinearTargetVelocityMultiplier=true
        bEnableMaxForce=true
        bEnableMaxTorque=true
        bEnablebDisableCollision=true
        bEnablebEnabled=true
        bEnablebOnlyControlChildObject=true
        bEnablebUseAccelerationDriveMode=true
        bEnablebUseCustomControlPoint=true
        bEnablebUseSkeletalAnimation=true
        bEnabled=true
        bOnlyControlChildObject=false
        bUseAccelerationDriveMode=true
        bUseCustomControlPoint=false
        bUseSkeletalAnimation=true
        linearDampingRatio=0.7
        linearExtraDamping=0
        linearStrength=4
        linearTargetVelocityMultiplier=1
        maxForce=0
        maxTorque=0
        customControlPoint=<0 0 0>
       }}
      Sleeve_L_ControlSet {data {
        angularDampingRatio=0.7
        angularExtraDamping=0
        angularStrength=4
        angularTargetVelocityMultiplier=1
        bDisableCollision=false
        bEnableAngularDampingRatio=true
        bEnableAngularExtraDamping=true
        bEnableAngularStrength=true
        bEnableAngularTargetVelocityMultiplier=true
        bEnableCustomControlPoint=true
        bEnableLinearDampingRatio=true
        bEnableLinearExtraDamping=true
        bEnableLinearStrength=true
        bEnableLinearTargetVelocityMultiplier=true
        bEnableMaxForce=true
        bEnableMaxTorque=true
        bEnablebDisableCollision=true
        bEnablebEnabled=true
        bEnablebOnlyControlChildObject=true
        bEnablebUseAccelerationDriveMode=true
        bEnablebUseCustomControlPoint=true
        bEnablebUseSkeletalAnimation=true
        bEnabled=true
        bOnlyControlChildObject=false
        bUseAccelerationDriveMode=true
        bUseCustomControlPoint=false
        bUseSkeletalAnimation=true
        linearDampingRatio=0.7
        linearExtraDamping=0
        linearStrength=4
        linearTargetVelocityMultiplier=1
        maxForce=0
        maxTorque=0
        customControlPoint=<0 0 0>
       }}
      SpineControlSet {data {
        angularDampingRatio=0.7
        angularExtraDamping=0
        angularStrength=3
        angularTargetVelocityMultiplier=1
        bDisableCollision=false
        bEnableAngularDampingRatio=true
        bEnableAngularExtraDamping=true
        bEnableAngularStrength=true
        bEnableAngularTargetVelocityMultiplier=true
        bEnableCustomControlPoint=true
        bEnableLinearDampingRatio=true
        bEnableLinearExtraDamping=true
        bEnableLinearStrength=true
        bEnableLinearTargetVelocityMultiplier=true
        bEnableMaxForce=true
        bEnableMaxTorque=true
        bEnablebDisableCollision=true
        bEnablebEnabled=true
        bEnablebOnlyControlChildObject=true
        bEnablebUseAccelerationDriveMode=true
        bEnablebUseCustomControlPoint=true
        bEnablebUseSkeletalAnimation=true
        bEnabled=true
        bOnlyControlChildObject=false
        bUseAccelerationDriveMode=true
        bUseCustomControlPoint=false
        bUseSkeletalAnimation=true
        linearDampingRatio=0.7
        linearExtraDamping=0
        linearStrength=3
        linearTargetVelocityMultiplier=1
        maxForce=0
        maxTorque=0
        customControlPoint=<0 0 0>
       }}
     ]
     modifierParameters [
      HeadSet {data {
        bEnableCCD=false
        bEnableCollisionType=true
        bEnableGravityMultiplier=true
        bEnableKinematicTargetSpace=true
        bEnableMovementType=true
        bEnablePhysicsBlendWeight=true
        bEnablebEnableCCD=true
        bEnablebUpdateKinematicFromSimulation=true
        bUpdateKinematicFromSimulation=true
        collisionType=QueryAndPhysics
        gravityMultiplier=1
        kinematicTargetSpace=OffsetInBoneSpace
        movementType=Simulated
        physicsBlendWeight=1
       }}
      Arm_L_Set {data {
        bEnableCCD=false
        bEnableCollisionType=true
        bEnableGravityMultiplier=true
        bEnableKinematicTargetSpace=true
        bEnableMovementType=true
        bEnablePhysicsBlendWeight=true
        bEnablebEnableCCD=true
        bEnablebUpdateKinematicFromSimulation=true
        bUpdateKinematicFromSimulation=true
        collisionType=QueryAndPhysics
        gravityMultiplier=1
        kinematicTargetSpace=OffsetInBoneSpace
        movementType=Simulated
        physicsBlendWeight=1
       }}
      Arm_R_Set {data {
        bEnableCCD=false
        bEnableCollisionType=true
        bEnableGravityMultiplier=true
        bEnableKinematicTargetSpace=true
        bEnableMovementType=true
        bEnablePhysicsBlendWeight=true
        bEnablebEnableCCD=true
        bEnablebUpdateKinematicFromSimulation=true
        bUpdateKinematicFromSimulation=true
        collisionType=QueryAndPhysics
        gravityMultiplier=1
        kinematicTargetSpace=OffsetInBoneSpace
        movementType=Simulated
        physicsBlendWeight=1
       }}
      Sleeve_R_Set {data {
        bEnableCCD=false
        bEnableCollisionType=true
        bEnableGravityMultiplier=true
        bEnableKinematicTargetSpace=true
        bEnableMovementType=true
        bEnablePhysicsBlendWeight=true
        bEnablebEnableCCD=true
        bEnablebUpdateKinematicFromSimulation=true
        bUpdateKinematicFromSimulation=true
        collisionType=QueryAndPhysics
        gravityMultiplier=1
        kinematicTargetSpace=OffsetInBoneSpace
        movementType=Simulated
        physicsBlendWeight=1
       }}
      Sleeve_L_Set {data {
        bEnableCCD=false
        bEnableCollisionType=true
        bEnableGravityMultiplier=true
        bEnableKinematicTargetSpace=true
        bEnableMovementType=true
        bEnablePhysicsBlendWeight=true
        bEnablebEnableCCD=true
        bEnablebUpdateKinematicFromSimulation=true
        bUpdateKinematicFromSimulation=true
        collisionType=QueryAndPhysics
        gravityMultiplier=1
        kinematicTargetSpace=OffsetInBoneSpace
        movementType=Simulated
        physicsBlendWeight=1
       }}
      SpineSet {data {
        bEnableCCD=false
        bEnableCollisionType=true
        bEnableGravityMultiplier=true
        bEnableKinematicTargetSpace=true
        bEnableMovementType=true
        bEnablePhysicsBlendWeight=true
        bEnablebEnableCCD=true
        bEnablebUpdateKinematicFromSimulation=true
        bUpdateKinematicFromSimulation=true
        collisionType=QueryAndPhysics
        gravityMultiplier=1
        kinematicTargetSpace=OffsetInBoneSpace
        movementType=Simulated
        physicsBlendWeight=1
       }}
     ]
    }
    overrideWorldGravity {z=-980}
    simSpaceSettings {maxAngularAcceleration=1, maxAngularVelocity=10, maxLinearAcceleration=1, maxLinearVelocity=20000, worldAlpha=1, externalLinearDragV=<2 2>}
   }
   AnimGraphNode_StateMachine_1 = Idle and Reactive() {is=title AnimGraphNode_StateMachine}
   n2 = "Transform (Modify) Bone - Bone: root"(ComponentPose=Local To Component(LocalPose=@AnimGraphNode_StateMachine_1), Translation="0,0,9", Rotation="0,0,0") {
    rotationSpace=BCS_BoneSpace
    translationMode=BMM_Additive
    translationSpace=BCS_BoneSpace
    alphaBoolBlend {blendOption=Linear}
    boneToModify {boneName=root}
    translation {z=9}
   }
   n3 = "Transform (Modify) Bone - Bone: bag_edge_01"!(ComponentPose=@n2, bAlphaBoolEnabled:bool=@n1, Translation="0,-0.75,0", Rotation="0,0,0", Scale="1,1,1") -> (Pose) {
    alphaInputType=Bool
    rotationMode=BMM_Additive
    rotationSpace=BCS_BoneSpace
    translationMode=BMM_Additive
    translationSpace=BCS_BoneSpace
    alphaBoolBlend {blendOption=Linear}
    boneToModify {boneName=bag_edge_01}
    translation {y=-0.75}
   }
   n4 = "Transform (Modify) Bone - Bone: bag_edge_04"!(ComponentPose=@n3, bAlphaBoolEnabled:bool=@n1, Translation="0,-0.7,0", Rotation="0,0,0", Scale="1,1,1") -> (Pose) {
    alphaInputType=Bool
    rotationMode=BMM_Additive
    rotationSpace=BCS_BoneSpace
    translationMode=BMM_Additive
    translationSpace=BCS_BoneSpace
    alphaBoolBlend {blendOption=Linear}
    boneToModify {boneName=bag_edge_04}
    translation {y=-0.7}
   }
   n5 = Get Gravity Override()
   n6 = RigidBody!(ComponentPose=@n4, AlphaCurveName:Name=PhysicsAlpha, OverrideWorldGravity:Vector=@n5, ExternalForce) -> (Pose) {
    alphaInputType=Curve
    bOverrideWorldGravity=true
    overridePhysicsAsset=&/M_MED_WidePlane_Pack_Ears_Physics
    simulationSpace=BaseBoneSpace
    alphaBoolBlend {blendOption=Linear}
    baseBoneRef {boneName=pack_spine_03}
    componentAppliedLinearAccClamp=<3000 3000 3000>
    componentLinearAccScale=<0.4 0.4 0.4>
    componentLinearVelScale=<0.5 0.5 0.5>
    simSpaceSettings {maxAngularAcceleration=100, maxAngularVelocity=10, maxLinearAcceleration=300, maxLinearVelocity=300, worldAlpha=0.7}
   }
   n7 = "Transform (Modify) Bone - Bone: pack"(ComponentPose=@n6, Translation=Get backpack_translation(), Rotation="0,0,0") {rotationMode=BMM_Additive, translationMode=BMM_Additive, alphaBoolBlend {blendOption=Linear}, boneToModify {boneName=pack}}
   n8 = "Transform (Modify) Bone - Bone: bag_edge_01"!(ComponentPose=@n7, Alpha:float=1, Translation="0.000,-0.240,-0.330", Rotation="0.005,-0.001,-7.844", Scale="1,1,1") -> (Pose) {
    rotationMode=BMM_Additive
    rotationSpace=BCS_BoneSpace
    translationMode=BMM_Additive
    translationSpace=BCS_BoneSpace
    alphaBoolBlend {blendOption=Linear}
    boneToModify {boneName=bag_edge_01}
    rotation=<rot 0.005 -0.001 -7.844>
    translation {y=-0.24, z=-0.33}
   }
   n9 = "Transform (Modify) Bone - Bone: bag_edge_04"!(ComponentPose=@n8, Alpha:float=1, Translation="0,-0.24,0", Rotation="0,0,0", Scale="1,1,1") -> (Pose) {
    rotationMode=BMM_Additive
    rotationSpace=BCS_BoneSpace
    translationMode=BMM_Additive
    translationSpace=BCS_BoneSpace
    alphaBoolBlend {blendOption=Linear}
    boneToModify {boneName=bag_edge_04}
    translation {y=-0.24}
   }
   n10 = "Transform (Modify) Bone - Bone: pack_hand_r"(ComponentPose=@n9, Alpha=1, Translation="0,0,0", Rotation="3.776,-12.956,14.613", Scale="1,1,1") {
    rotationMode=BMM_Additive
    rotationSpace=BCS_BoneSpace
    translationMode=BMM_Additive
    translationSpace=BCS_BoneSpace
    alphaBoolBlend {blendOption=Linear}
    boneToModify {boneName=pack_hand_r}
    rotation=<rot 3.776 -12.956 14.613>
   }
   n11 = Anim Dynamics - Bone: dyn_hair_01!(ComponentPose=@n10, Alpha=1, GravityOverride:Vector=@n5, ExternalForce) -> (Pose) {
    angularDampingOverride=0.99
    angularSpringConstant=2000
    bAngularSpring=true
    bOverrideAngularDamping=true
    bUseGravityOverride=true
    simulationSpace=BoneRelative
    alphaBoolBlend {blendOption=Linear}
    boundBone {boneName=dyn_hair_01}
    componentAppliedLinearAccClamp=<3000 3000 3000>
    componentLinearAccScale=<0.4 0.4 0.4>
    componentLinearVelScale=<0.5 0.5 0.5>
    physicsBodyDefinitions [{
      boundBone {boneName=dyn_hair_01}
      boxExtents=<16 16 16>
      constraintSetup {angularLimitsMax=<0 8 8>, angularLimitsMin=<0 -8 -8>, angularTarget=<1 0 0>}
      localJointOffset=<0.2 0 0>
     }]
    relativeSpaceBone {boneName=root}
   }
   n12 = Anim Dynamics - Bone: dyn_hair_02!(ComponentPose=@n11, Alpha=1, GravityOverride:Vector=@n5, ExternalForce) -> (Pose) {
    angularDampingOverride=0.99
    angularSpringConstant=2000
    bAngularSpring=true
    bOverrideAngularDamping=true
    bUseGravityOverride=true
    simulationSpace=BoneRelative
    alphaBoolBlend {blendOption=Linear}
    boundBone {boneName=dyn_hair_02}
    componentAppliedLinearAccClamp=<3000 3000 3000>
    componentLinearAccScale=<0.4 0.4 0.4>
    componentLinearVelScale=<0.5 0.5 0.5>
    physicsBodyDefinitions [{
      boundBone {boneName=dyn_hair_02}
      boxExtents=<16 16 16>
      constraintSetup {angularLimitsMax=<0 8 8>, angularLimitsMin=<0 -8 -8>, angularTarget=<1 0 0>}
      localJointOffset=<0.15 0 0>
     }]
    relativeSpaceBone {boneName=root}
   }
   n13 = Anim Dynamics - Bone: dyn_hair_03!(ComponentPose=@n12, Alpha=1, GravityOverride:Vector=@n5, ExternalForce="0,0,0") -> (Pose) {
    angularDampingOverride=0.99
    angularSpringConstant=2000
    bAngularSpring=true
    bOverrideAngularDamping=true
    bUseGravityOverride=true
    simulationSpace=BoneRelative
    alphaBoolBlend {blendOption=Linear}
    boundBone {boneName=dyn_hair_03}
    componentAppliedLinearAccClamp=<3000 3000 3000>
    componentLinearAccScale=<0.4 0.4 0.4>
    componentLinearVelScale=<0.5 0.5 0.5>
    physicsBodyDefinitions [{
      boundBone {boneName=dyn_hair_03}
      boxExtents=<16 16 16>
      constraintSetup {angularLimitsMax=<0 8 8>, angularLimitsMin=<0 -8 -8>, angularTarget=<1 0 0>}
      localJointOffset=<0.3 0 0>
     }]
    relativeSpaceBone {boneName=root}
   }
   n14 = Anim Dynamics - Bone: dyn_hair_04!(ComponentPose=@n13, Alpha=1, GravityOverride:Vector=@n5, ExternalForce="0,0,0") -> (Pose) {
    angularDampingOverride=0.99
    angularSpringConstant=2000
    bAngularSpring=true
    bOverrideAngularDamping=true
    bUseGravityOverride=true
    simulationSpace=BoneRelative
    alphaBoolBlend {blendOption=Linear}
    boundBone {boneName=dyn_hair_04}
    componentAppliedLinearAccClamp=<3000 3000 3000>
    componentLinearAccScale=<0.4 0.4 0.4>
    componentLinearVelScale=<0.5 0.5 0.5>
    physicsBodyDefinitions [{
      boundBone {boneName=dyn_hair_04}
      boxExtents=<16 16 16>
      constraintSetup {angularLimitsMax=<0 8 8>, angularLimitsMin=<0 -8 -8>, angularTarget=<1 0 0>}
      localJointOffset=<0.15 0 0>
     }]
    relativeSpaceBone {boneName=root}
   }
   n15 = Anim Dynamics - Bone: dyn_hair_05!(ComponentPose=@n14, Alpha=1, GravityOverride:Vector=@n5, ExternalForce="0,0,0") -> (Pose) {
    angularDampingOverride=0.99
    angularSpringConstant=2000
    bAngularSpring=true
    bOverrideAngularDamping=true
    bUseGravityOverride=true
    simulationSpace=BoneRelative
    alphaBoolBlend {blendOption=Linear}
    boundBone {boneName=dyn_hair_05}
    componentAppliedLinearAccClamp=<3000 3000 3000>
    componentLinearAccScale=<0.4 0.4 0.4>
    componentLinearVelScale=<0.5 0.5 0.5>
    physicsBodyDefinitions [{
      boundBone {boneName=dyn_hair_05}
      boxExtents=<16 16 16>
      constraintSetup {angularLimitsMax=<0 8 8>, angularLimitsMin=<0 -8 -8>, angularTarget=<1 0 0>}
      localJointOffset=<0.17 0 0>
     }]
    relativeSpaceBone {boneName=root}
   }
   n16 = Anim Dynamics - Bone: pack!(ComponentPose=@n15, AlphaCurveName:Name=PhysicsAlpha, GravityOverride:Vector=@n5, ExternalForce="0,0,0") -> (Pose) {
    alphaInputType=Curve
    angularDampingOverride=0.9
    angularSpringConstant=7500
    bAngularSpring=true
    bOverrideAngularDamping=true
    bUseGravityOverride=true
    simulationSpace=BoneRelative
    alphaBoolBlend {blendOption=Linear}
    boundBone {boneName=pack}
    componentAppliedLinearAccClamp=<5000 5000 5000>
    componentLinearAccScale=<0.3 0.3 0.3>
    componentLinearVelScale=<0.4 0.4 0.4>
    physicsBodyDefinitions [{
      boundBone {boneName=pack}
      boxExtents=<20 20 20>
      constraintSetup {angularLimitsMax=<0 0 10>, angularLimitsMin=<0 0 -10>, angularTarget=<0 0 1>}
      localJointOffset=<0 0.2 0>
     }]
    relativeSpaceBone {boneName=root}
   }
   Output Pose(Result=Component To Local(ComponentPose=@n16)) {layerGroup=None}
   AnimGraphNode_StateMachine_0 = Cane/NoCane States() {is=title AnimGraphNode_StateMachine}
   Get backpack_rotate()
  ]}
 Cane/NoCane States {
  entry=Idle_Cane
  path=AnimGraph.AnimGraphNode_StateMachine_0.Cane/NoCane States
  ids=id
  nodes [
   AnimStateNode_0 = Idle_NoCane() {is=title AnimStateNode}
   AnimStateNode_4 = Idle_Cane() {is=title AnimStateNode}
   AnimStateTransitionNode_1 = AnimStateTransitionNode()
   AnimStateTransitionNode_3 = AnimStateTransitionNode()
  ]
  links [
   {
    dst=Idle_NoCane
    rule=AnimGraph.AnimGraphNode_StateMachine_0.Cane/NoCane States.AnimStateTransitionNode_3.Transition
    src=Idle_Cane
    via=AnimStateTransitionNode_3
   }
   {
    dst=Idle_Cane
    rule=AnimGraph.AnimGraphNode_StateMachine_0.Cane/NoCane States.AnimStateTransitionNode_1.Transition
    src=Idle_NoCane
    via=AnimStateTransitionNode_1
   }
  ]
 }
 Idle_NoCane {
  path=AnimGraph.AnimGraphNode_StateMachine_0.Cane/NoCane States.AnimStateNode_0.Idle_NoCane
  nodes [
   AnimGraphNode_StateMachine_1 = "Idle && Reactive (NoCane)"() {is=title AnimGraphNode_StateMachine}
   Output Animation Pose(Result=@AnimGraphNode_StateMachine_1)
  ]
 }
 "Idle && Reactive (NoCane)" {
  entry=Idle
  path="AnimGraph.AnimGraphNode_StateMachine_0.Cane/NoCane States.AnimStateNode_0.Idle_NoCane.AnimGraphNode_StateMachine_1.Idle && Reactive (NoCane)"
  ids=id
  nodes [
   AnimStateNode_0 = Chest_Open() {
    is=title AnimStateNode
    stateEntered {guid=#1}
    stateFullyBlended {guid=#2}
    stateLeft {guid=#3}
   }
   AnimStateNode_2 = Idle() {
    stateEntered {guid=#4}
    stateFullyBlended {guid=#5}
    stateLeft {guid=#6}
   }
   AnimStateNode_3 = ChugJug() {
    is=title AnimStateNode
    stateEntered {guid=#1}
    stateFullyBlended {guid=#2}
    stateLeft {guid=#3}
   }
   AnimStateTransitionNode_0 = AnimStateTransitionNode()
   AnimStateTransitionNode_1 = AnimStateTransitionNode() {bAutomaticRuleBasedOnSequencePlayerInState=true}
   AnimStateTransitionNode_2 = AnimStateTransitionNode() {
    crossfadeDuration=1
    transitionEnd {guid=#7}
    transitionInterrupt {guid=#8}
    transitionStart {guid=#9}
   }
   AnimStateTransitionNode_3 = AnimStateTransitionNode() {
    bAutomaticRuleBasedOnSequencePlayerInState=true
    crossfadeDuration=0.1
    transitionEnd {guid=#10}
    transitionInterrupt {guid=#11}
    transitionStart {guid=#12}
   }
  ]
  links [
   {
    dst=Idle
    rule="AnimGraph.AnimGraphNode_StateMachine_0.Cane/NoCane States.AnimStateNode_0.Idle_NoCane.AnimGraphNode_StateMachine_1.Idle && Reactive (NoCane).AnimStateTransitionNode_3.Transition"
    src=Chest_Open
    via=AnimStateTransitionNode_3
   }
   {
    dst=Idle
    rule="AnimGraph.AnimGraphNode_StateMachine_0.Cane/NoCane States.AnimStateNode_0.Idle_NoCane.AnimGraphNode_StateMachine_1.Idle && Reactive (NoCane).AnimStateTransitionNode_1.Transition"
    src=ChugJug
    via=AnimStateTransitionNode_1
   }
   {
    dst=Chest_Open
    rule="AnimGraph.AnimGraphNode_StateMachine_0.Cane/NoCane States.AnimStateNode_0.Idle_NoCane.AnimGraphNode_StateMachine_1.Idle && Reactive (NoCane).AnimStateTransitionNode_2.Transition"
    src=Idle
    via=AnimStateTransitionNode_2
   }
   {
    dst=ChugJug
    rule="AnimGraph.AnimGraphNode_StateMachine_0.Cane/NoCane States.AnimStateNode_0.Idle_NoCane.AnimGraphNode_StateMachine_1.Idle && Reactive (NoCane).AnimStateTransitionNode_0.Transition"
    src=Idle
    via=AnimStateTransitionNode_0
   }
  ]
 }
 Chest_Open {
  path="AnimGraph.AnimGraphNode_StateMachine_0.Cane/NoCane States.AnimStateNode_0.Idle_NoCane.AnimGraphNode_StateMachine_1.Idle && Reactive (NoCane).AnimStateNode_0.Chest_Open"
  nodes [
   n1 = Sequence Player 'WidePlane_Reactive_NoCaneCurious_PACK'() {
    bLoopAnimation=false
    groupName=Curious
    method=SyncGroup
    sequence=/BRCosmetics/Accessories/FORT_Backpacks/Backpack_M_MED_WidePlane/Anim/WidePlane_Reactive_NoCaneCurious_PACK
   }
   Output Animation Pose(Result=@n1)
  ]
 }
 Idle {
  path="AnimGraph.AnimGraphNode_StateMachine_0.Cane/NoCane States.AnimStateNode_0.Idle_NoCane.AnimGraphNode_StateMachine_1.Idle && Reactive (NoCane).AnimStateNode_2.Idle"
  nodes [
   AnimGraphNode_StateMachine_2 = "Idle (NoCane)"() {is=title AnimGraphNode_StateMachine}
   Output Animation Pose(Result=@AnimGraphNode_StateMachine_2)
  ]
 }
 "Idle (NoCane)" {
  entry=Idle_NoCane
  path="AnimGraph.AnimGraphNode_StateMachine_0.Cane/NoCane States.AnimStateNode_0.Idle_NoCane.AnimGraphNode_StateMachine_1.Idle && Reactive (NoCane).AnimStateNode_2.Idle.AnimGraphNode_StateMachine_2.Idle (NoCane)"
  ids=id
  nodes [
   AnimStateNode_0 = Idle_NoCane()
   AnimStateNode_1 = Meditate_NoCane() {is=title AnimStateNode, stateEntered {guid=#13, notifyName=EnteredNoCaneMeditate}}
   AnimStateTransitionNode_4 = AnimStateTransitionNode()
   AnimStateTransitionNode_5 = AnimStateTransitionNode() {bAutomaticRuleBasedOnSequencePlayerInState=true}
  ]
  links [
   {
    dst=Meditate_NoCane
    rule="AnimGraph.AnimGraphNode_StateMachine_0.Cane/NoCane States.AnimStateNode_0.Idle_NoCane.AnimGraphNode_StateMachine_1.Idle && Reactive (NoCane).AnimStateNode_2.Idle.AnimGraphNode_StateMachine_2.Idle (NoCane).AnimStateTransitionNode_4.Transition"
    src=Idle_NoCane
    via=AnimStateTransitionNode_4
   }
   {
    dst=Idle_NoCane
    rule="AnimGraph.AnimGraphNode_StateMachine_0.Cane/NoCane States.AnimStateNode_0.Idle_NoCane.AnimGraphNode_StateMachine_1.Idle && Reactive (NoCane).AnimStateNode_2.Idle.AnimGraphNode_StateMachine_2.Idle (NoCane).AnimStateTransitionNode_5.Transition"
    src=Meditate_NoCane
    via=AnimStateTransitionNode_5
   }
  ]
 }
 Idle_NoCane {
  path="AnimGraph.AnimGraphNode_StateMachine_0.Cane/NoCane States.AnimStateNode_0.Idle_NoCane.AnimGraphNode_StateMachine_1.Idle && Reactive (NoCane).AnimStateNode_2.Idle.AnimGraphNode_StateMachine_2.Idle (NoCane).AnimStateNode_0.Idle_NoCane"
  nodes [
   n1 = Sequence Player 'WidePlane_Idle_NoCane_PACK'() {groupName=Idle, method=SyncGroup, sequence=/BRCosmetics/Accessories/FORT_Backpacks/Backpack_M_MED_WidePlane/Anim/WidePlane_Idle_NoCane_PACK}
   Output Animation Pose(Result=@n1)
  ]
 }
 Meditate_NoCane {
  path="AnimGraph.AnimGraphNode_StateMachine_0.Cane/NoCane States.AnimStateNode_0.Idle_NoCane.AnimGraphNode_StateMachine_1.Idle && Reactive (NoCane).AnimStateNode_2.Idle.AnimGraphNode_StateMachine_2.Idle (NoCane).AnimStateNode_1.Meditate_NoCane"
  nodes [
   n1 = Sequence Player 'WidePlane_IdleBreak_Meditate_NoCane_PACK'() {
    bLoopAnimation=false
    groupName=Meditate
    method=SyncGroup
    sequence=/BRCosmetics/Accessories/FORT_Backpacks/Backpack_M_MED_WidePlane/Anim/WidePlane_IdleBreak_Meditate_NoCane_PACK
   }
   Output Animation Pose(Result=@n1)
  ]
 }
 Transition {
  path="AnimGraph.AnimGraphNode_StateMachine_0.Cane/NoCane States.AnimStateNode_0.Idle_NoCane.AnimGraphNode_StateMachine_1.Idle && Reactive (NoCane).AnimStateNode_2.Idle.AnimGraphNode_StateMachine_2.Idle (NoCane).AnimStateTransitionNode_4.Transition"
  nodes [
   n1 = "Get Relevant Anim Time Remaining (Idle_NoCane)"!() -> (ReturnValue:float) {is=title K2Node_AnimGetter}
   Result(bCanEnterTransition=AND Boolean(A=Get UseTheForce() -> (UseTheForce:bool), B="float <= float"(A=@n1, B=0.2) -> (ReturnValue=~)))
  ]
 }
 Transition {
  path="AnimGraph.AnimGraphNode_StateMachine_0.Cane/NoCane States.AnimStateNode_0.Idle_NoCane.AnimGraphNode_StateMachine_1.Idle && Reactive (NoCane).AnimStateNode_2.Idle.AnimGraphNode_StateMachine_2.Idle (NoCane).AnimStateTransitionNode_5.Transition"
  nodes [Result()]
 }
 ChugJug {
  path="AnimGraph.AnimGraphNode_StateMachine_0.Cane/NoCane States.AnimStateNode_0.Idle_NoCane.AnimGraphNode_StateMachine_1.Idle && Reactive (NoCane).AnimStateNode_3.ChugJug"
  nodes [
   n1 = Sequence Player 'WidePlane_Reactive_NoCaneLaugh_PACK'() {
    bLoopAnimation=false
    groupName=Laugh
    method=SyncGroup
    sequence=/BRCosmetics/Accessories/FORT_Backpacks/Backpack_M_MED_WidePlane/Anim/WidePlane_Reactive_NoCaneLaugh_PACK
   }
   Output Animation Pose(Result=@n1)
  ]
 }
 Transition {
  path="AnimGraph.AnimGraphNode_StateMachine_0.Cane/NoCane States.AnimStateNode_0.Idle_NoCane.AnimGraphNode_StateMachine_1.Idle && Reactive (NoCane).AnimStateTransitionNode_0.Transition"
  nodes [Result(bCanEnterTransition=Get ChugJug Animation() -> (ChugJug Animation:bool))]
 }
 Transition {
  path="AnimGraph.AnimGraphNode_StateMachine_0.Cane/NoCane States.AnimStateNode_0.Idle_NoCane.AnimGraphNode_StateMachine_1.Idle && Reactive (NoCane).AnimStateTransitionNode_1.Transition"
  nodes [Result()]
 }
 Transition {
  path="AnimGraph.AnimGraphNode_StateMachine_0.Cane/NoCane States.AnimStateNode_0.Idle_NoCane.AnimGraphNode_StateMachine_1.Idle && Reactive (NoCane).AnimStateTransitionNode_2.Transition"
  nodes [Result(bCanEnterTransition=Get ChestOpen Animation() -> (ChestOpen Animation:bool))]
 }
 Transition {
  path="AnimGraph.AnimGraphNode_StateMachine_0.Cane/NoCane States.AnimStateNode_0.Idle_NoCane.AnimGraphNode_StateMachine_1.Idle && Reactive (NoCane).AnimStateTransitionNode_3.Transition"
  nodes [Result()]
 }
 Idle_Cane {
  path=AnimGraph.AnimGraphNode_StateMachine_0.Cane/NoCane States.AnimStateNode_4.Idle_Cane
  nodes [
   Sequence Player 'WidePlane_Idle_Cane_PACK'() {sequence=/BRCosmetics/Accessories/FORT_Backpacks/Backpack_M_MED_WidePlane/Anim/WidePlane_Idle_Cane_PACK}
   AnimGraphNode_StateMachine_1 = "Idle && Reactive (Cane)"() {is=title AnimGraphNode_StateMachine}
   Output Animation Pose(Result=@AnimGraphNode_StateMachine_1)
  ]
 }
 "Idle && Reactive (Cane)" {
  entry=Idle
  path="AnimGraph.AnimGraphNode_StateMachine_0.Cane/NoCane States.AnimStateNode_4.Idle_Cane.AnimGraphNode_StateMachine_1.Idle && Reactive (Cane)"
  ids=id
  nodes [
   AnimStateNode_0 = Chest_Open() {
    stateEntered {guid=#1}
    stateFullyBlended {guid=#2, notifyName=IsInspectingChest}
    stateLeft {guid=#3}
   }
   AnimStateNode_2 = Idle() {
    stateEntered {guid=#4}
    stateFullyBlended {guid=#5}
    stateLeft {guid=#6}
   }
   AnimStateNode_3 = ChugJug() {
    stateEntered {guid=#1}
    stateFullyBlended {guid=#2}
    stateLeft {guid=#3}
   }
   AnimStateTransitionNode_0 = AnimStateTransitionNode()
   AnimStateTransitionNode_2 = AnimStateTransitionNode() {
    crossfadeDuration=0.3
    transitionEnd {guid=#7}
    transitionInterrupt {guid=#8}
    transitionStart {guid=#9}
   }
   AnimStateTransitionNode_3 = AnimStateTransitionNode() {
    bAutomaticRuleBasedOnSequencePlayerInState=true
    transitionEnd {guid=#10}
    transitionInterrupt {guid=#11}
    transitionStart {guid=#12}
   }
   AnimStateTransitionNode_5 = AnimStateTransitionNode() {bAutomaticRuleBasedOnSequencePlayerInState=true}
  ]
  links [
   {
    dst=Idle
    rule="AnimGraph.AnimGraphNode_StateMachine_0.Cane/NoCane States.AnimStateNode_4.Idle_Cane.AnimGraphNode_StateMachine_1.Idle && Reactive (Cane).AnimStateTransitionNode_3.Transition"
    src=Chest_Open
    via=AnimStateTransitionNode_3
   }
   {
    dst=Idle
    rule="AnimGraph.AnimGraphNode_StateMachine_0.Cane/NoCane States.AnimStateNode_4.Idle_Cane.AnimGraphNode_StateMachine_1.Idle && Reactive (Cane).AnimStateTransitionNode_5.Transition"
    src=ChugJug
    via=AnimStateTransitionNode_5
   }
   {
    dst=Chest_Open
    rule="AnimGraph.AnimGraphNode_StateMachine_0.Cane/NoCane States.AnimStateNode_4.Idle_Cane.AnimGraphNode_StateMachine_1.Idle && Reactive (Cane).AnimStateTransitionNode_2.Transition"
    src=Idle
    via=AnimStateTransitionNode_2
   }
   {
    dst=ChugJug
    rule="AnimGraph.AnimGraphNode_StateMachine_0.Cane/NoCane States.AnimStateNode_4.Idle_Cane.AnimGraphNode_StateMachine_1.Idle && Reactive (Cane).AnimStateTransitionNode_0.Transition"
    src=Idle
    via=AnimStateTransitionNode_0
   }
  ]
 }
 Chest_Open {
  path="AnimGraph.AnimGraphNode_StateMachine_0.Cane/NoCane States.AnimStateNode_4.Idle_Cane.AnimGraphNode_StateMachine_1.Idle && Reactive (Cane).AnimStateNode_0.Chest_Open"
  nodes [
   n1 = Sequence Player 'WidePlane_Reactive_CaneCurious_PACK'() {
    bLoopAnimation=false
    groupName=Curious
    method=SyncGroup
    sequence=/BRCosmetics/Accessories/FORT_Backpacks/Backpack_M_MED_WidePlane/Anim/WidePlane_Reactive_CaneCurious_PACK
   }
   Output Animation Pose(Result=@n1)
  ]
 }
 Idle {
  path="AnimGraph.AnimGraphNode_StateMachine_0.Cane/NoCane States.AnimStateNode_4.Idle_Cane.AnimGraphNode_StateMachine_1.Idle && Reactive (Cane).AnimStateNode_2.Idle"
  nodes [
   AnimGraphNode_StateMachine_0 = ANIM_TEST_IDLE() {is=title AnimGraphNode_StateMachine}
   AnimGraphNode_StateMachine_1 = "Idle (Cane)"() {is=title AnimGraphNode_StateMachine}
   Output Animation Pose(Result=@AnimGraphNode_StateMachine_1)
  ]
 }
 ANIM_TEST_IDLE {
  entry=IDLE
  path="AnimGraph.AnimGraphNode_StateMachine_0.Cane/NoCane States.AnimStateNode_4.Idle_Cane.AnimGraphNode_StateMachine_1.Idle && Reactive (Cane).AnimStateNode_2.Idle.AnimGraphNode_StateMachine_0.ANIM_TEST_IDLE"
  ids=id
  nodes [
   AnimStateNode_0 = IDLE() {is=title AnimStateNode}
   AnimStateNode_1 = IDLE_REPEAT() {is=title AnimStateNode}
   AnimStateNode_2 = IDLEBREAK_MEDITATE() {is=title AnimStateNode}
   AnimStateNode_3 = REACTIVE_SLURP() {is=title AnimStateNode}
   AnimStateNode_4 = REACTIVE_OPEN_CHEST() {is=title AnimStateNode}
   AnimStateTransitionNode_1 = AnimStateTransitionNode() {bAutomaticRuleBasedOnSequencePlayerInState=true}
   AnimStateTransitionNode_2 = AnimStateTransitionNode() {bAutomaticRuleBasedOnSequencePlayerInState=true}
   AnimStateTransitionNode_3 = AnimStateTransitionNode() {bAutomaticRuleBasedOnSequencePlayerInState=true}
   AnimStateTransitionNode_4 = AnimStateTransitionNode() {bAutomaticRuleBasedOnSequencePlayerInState=true}
   AnimStateTransitionNode_5 = AnimStateTransitionNode() {bAutomaticRuleBasedOnSequencePlayerInState=true}
  ]
  links [
   {
    dst=IDLE_REPEAT
    rule="AnimGraph.AnimGraphNode_StateMachine_0.Cane/NoCane States.AnimStateNode_4.Idle_Cane.AnimGraphNode_StateMachine_1.Idle && Reactive (Cane).AnimStateNode_2.Idle.AnimGraphNode_StateMachine_0.ANIM_TEST_IDLE.AnimStateTransitionNode_4.AnimationTransitionGraph_0"
    src=IDLE
    via=AnimStateTransitionNode_4
   }
   {
    dst=REACTIVE_SLURP
    rule="AnimGraph.AnimGraphNode_StateMachine_0.Cane/NoCane States.AnimStateNode_4.Idle_Cane.AnimGraphNode_StateMachine_1.Idle && Reactive (Cane).AnimStateNode_2.Idle.AnimGraphNode_StateMachine_0.ANIM_TEST_IDLE.AnimStateTransitionNode_1.AnimationTransitionGraph_1"
    src=IDLEBREAK_MEDITATE
    via=AnimStateTransitionNode_1
   }
   {
    dst=IDLEBREAK_MEDITATE
    rule="AnimGraph.AnimGraphNode_StateMachine_0.Cane/NoCane States.AnimStateNode_4.Idle_Cane.AnimGraphNode_StateMachine_1.Idle && Reactive (Cane).AnimStateNode_2.Idle.AnimGraphNode_StateMachine_0.ANIM_TEST_IDLE.AnimStateTransitionNode_5.AnimationTransitionGraph_0"
    src=IDLE_REPEAT
    via=AnimStateTransitionNode_5
   }
   {
    dst=IDLE
    rule="AnimGraph.AnimGraphNode_StateMachine_0.Cane/NoCane States.AnimStateNode_4.Idle_Cane.AnimGraphNode_StateMachine_1.Idle && Reactive (Cane).AnimStateNode_2.Idle.AnimGraphNode_StateMachine_0.ANIM_TEST_IDLE.AnimStateTransitionNode_3.AnimationTransitionGraph_1"
    src=REACTIVE_OPEN_CHEST
    via=AnimStateTransitionNode_3
   }
   {
    dst=REACTIVE_OPEN_CHEST
    rule="AnimGraph.AnimGraphNode_StateMachine_0.Cane/NoCane States.AnimStateNode_4.Idle_Cane.AnimGraphNode_StateMachine_1.Idle && Reactive (Cane).AnimStateNode_2.Idle.AnimGraphNode_StateMachine_0.ANIM_TEST_IDLE.AnimStateTransitionNode_2.Transition"
    src=REACTIVE_SLURP
    via=AnimStateTransitionNode_2
   }
  ]
 }
 IDLE {
  path="AnimGraph.AnimGraphNode_StateMachine_0.Cane/NoCane States.AnimStateNode_4.Idle_Cane.AnimGraphNode_StateMachine_1.Idle && Reactive (Cane).AnimStateNode_2.Idle.AnimGraphNode_StateMachine_0.ANIM_TEST_IDLE.AnimStateNode_0.IDLE"
  nodes [Output Animation Pose(Result=Sequence Player 'WidePlane_Idle_Cane_PACK'() {bLoopAnimation=false, sequence=&WidePlane_Idle_Cane_PACK})]
 }
 IDLE_REPEAT {
  path="AnimGraph.AnimGraphNode_StateMachine_0.Cane/NoCane States.AnimStateNode_4.Idle_Cane.AnimGraphNode_StateMachine_1.Idle && Reactive (Cane).AnimStateNode_2.Idle.AnimGraphNode_StateMachine_0.ANIM_TEST_IDLE.AnimStateNode_1.IDLE_REPEAT"
  nodes [Output Animation Pose(Result=Sequence Player 'WidePlane_Idle_Cane_PACK'() {bLoopAnimation=false, sequence=&WidePlane_Idle_Cane_PACK})]
 }
 IDLEBREAK_MEDITATE {
  path="AnimGraph.AnimGraphNode_StateMachine_0.Cane/NoCane States.AnimStateNode_4.Idle_Cane.AnimGraphNode_StateMachine_1.Idle && Reactive (Cane).AnimStateNode_2.Idle.AnimGraphNode_StateMachine_0.ANIM_TEST_IDLE.AnimStateNode_2.IDLEBREAK_MEDITATE"
  nodes [
   n1 = Sequence Player 'WidePlane_IdleBreak_Meditate_Cane_PACK'() {bLoopAnimation=false, sequence=/BRCosmetics/Accessories/FORT_Backpacks/Backpack_M_MED_WidePlane/Anim/WidePlane_IdleBreak_Meditate_Cane_PACK}
   Output Animation Pose(Result=@n1)
  ]
 }
 REACTIVE_SLURP {
  path="AnimGraph.AnimGraphNode_StateMachine_0.Cane/NoCane States.AnimStateNode_4.Idle_Cane.AnimGraphNode_StateMachine_1.Idle && Reactive (Cane).AnimStateNode_2.Idle.AnimGraphNode_StateMachine_0.ANIM_TEST_IDLE.AnimStateNode_3.REACTIVE_SLURP"
  nodes [
   n1 = Sequence Player 'WidePlane_Reactive_CaneLaugh_PACK'() {bLoopAnimation=false, sequence=/BRCosmetics/Accessories/FORT_Backpacks/Backpack_M_MED_WidePlane/Anim/WidePlane_Reactive_CaneLaugh_PACK}
   Output Animation Pose(Result=@n1)
  ]
 }
 REACTIVE_OPEN_CHEST {
  path="AnimGraph.AnimGraphNode_StateMachine_0.Cane/NoCane States.AnimStateNode_4.Idle_Cane.AnimGraphNode_StateMachine_1.Idle && Reactive (Cane).AnimStateNode_2.Idle.AnimGraphNode_StateMachine_0.ANIM_TEST_IDLE.AnimStateNode_4.REACTIVE_OPEN_CHEST"
  nodes [
   n1 = Sequence Player 'WidePlane_Reactive_CaneCurious_PACK'() {bLoopAnimation=false, sequence=&WidePlane_Reactive_CaneCurious_PACK}
   Output Animation Pose(Result=@n1)
  ]
 }
 AnimationTransitionGraph_1 {
  path="AnimGraph.AnimGraphNode_StateMachine_0.Cane/NoCane States.AnimStateNode_4.Idle_Cane.AnimGraphNode_StateMachine_1.Idle && Reactive (Cane).AnimStateNode_2.Idle.AnimGraphNode_StateMachine_0.ANIM_TEST_IDLE.AnimStateTransitionNode_1.AnimationTransitionGraph_1"
  nodes [Result()]
 }
 Transition {
  path="AnimGraph.AnimGraphNode_StateMachine_0.Cane/NoCane States.AnimStateNode_4.Idle_Cane.AnimGraphNode_StateMachine_1.Idle && Reactive (Cane).AnimStateNode_2.Idle.AnimGraphNode_StateMachine_0.ANIM_TEST_IDLE.AnimStateTransitionNode_2.Transition"
  nodes [Result()]
 }
 AnimationTransitionGraph_1 {
  path="AnimGraph.AnimGraphNode_StateMachine_0.Cane/NoCane States.AnimStateNode_4.Idle_Cane.AnimGraphNode_StateMachine_1.Idle && Reactive (Cane).AnimStateNode_2.Idle.AnimGraphNode_StateMachine_0.ANIM_TEST_IDLE.AnimStateTransitionNode_3.AnimationTransitionGraph_1"
  nodes [Result()]
 }
 AnimationTransitionGraph_0 {
  path="AnimGraph.AnimGraphNode_StateMachine_0.Cane/NoCane States.AnimStateNode_4.Idle_Cane.AnimGraphNode_StateMachine_1.Idle && Reactive (Cane).AnimStateNode_2.Idle.AnimGraphNode_StateMachine_0.ANIM_TEST_IDLE.AnimStateTransitionNode_4.AnimationTransitionGraph_0"
  nodes [Result()]
 }
 AnimationTransitionGraph_0 {
  path="AnimGraph.AnimGraphNode_StateMachine_0.Cane/NoCane States.AnimStateNode_4.Idle_Cane.AnimGraphNode_StateMachine_1.Idle && Reactive (Cane).AnimStateNode_2.Idle.AnimGraphNode_StateMachine_0.ANIM_TEST_IDLE.AnimStateTransitionNode_5.AnimationTransitionGraph_0"
  nodes [Result()]
 }
 "Idle (Cane)" {
  entry=Idle
  path="AnimGraph.AnimGraphNode_StateMachine_0.Cane/NoCane States.AnimStateNode_4.Idle_Cane.AnimGraphNode_StateMachine_1.Idle && Reactive (Cane).AnimStateNode_2.Idle.AnimGraphNode_StateMachine_1.Idle (Cane)"
  ids=id
  nodes [
   AnimStateNode_0 = Idle()
   AnimStateNode_1 = Meditate() {is=title AnimStateNode, stateEntered {guid=#14, notifyName=EnteredCaneMeditate}}
   AnimStateTransitionNode_2 = AnimStateTransitionNode()
   AnimStateTransitionNode_3 = AnimStateTransitionNode() {bAutomaticRuleBasedOnSequencePlayerInState=true}
  ]
  links [
   {
    dst=Meditate
    rule="AnimGraph.AnimGraphNode_StateMachine_0.Cane/NoCane States.AnimStateNode_4.Idle_Cane.AnimGraphNode_StateMachine_1.Idle && Reactive (Cane).AnimStateNode_2.Idle.AnimGraphNode_StateMachine_1.Idle (Cane).AnimStateTransitionNode_2.Transition"
    src=Idle
    via=AnimStateTransitionNode_2
   }
   {
    dst=Idle
    rule="AnimGraph.AnimGraphNode_StateMachine_0.Cane/NoCane States.AnimStateNode_4.Idle_Cane.AnimGraphNode_StateMachine_1.Idle && Reactive (Cane).AnimStateNode_2.Idle.AnimGraphNode_StateMachine_1.Idle (Cane).AnimStateTransitionNode_3.Transition"
    src=Meditate
    via=AnimStateTransitionNode_3
   }
  ]
 }
 Idle {
  path="AnimGraph.AnimGraphNode_StateMachine_0.Cane/NoCane States.AnimStateNode_4.Idle_Cane.AnimGraphNode_StateMachine_1.Idle && Reactive (Cane).AnimStateNode_2.Idle.AnimGraphNode_StateMachine_1.Idle (Cane).AnimStateNode_0.Idle"
  nodes [
   n1 = Sequence Player 'WidePlane_Idle_Cane_PACK'() {groupName=Idle, method=SyncGroup, sequence=&WidePlane_Idle_Cane_PACK}
   Output Animation Pose(Result=@n1)
  ]
 }
 Meditate {
  path="AnimGraph.AnimGraphNode_StateMachine_0.Cane/NoCane States.AnimStateNode_4.Idle_Cane.AnimGraphNode_StateMachine_1.Idle && Reactive (Cane).AnimStateNode_2.Idle.AnimGraphNode_StateMachine_1.Idle (Cane).AnimStateNode_1.Meditate"
  nodes [
   n1 = Sequence Player 'WidePlane_IdleBreak_Meditate_Cane_PACK'() {
    bLoopAnimation=false
    groupName=Meditate
    method=SyncGroup
    sequence=&WidePlane_IdleBreak_Meditate_Cane_PACK
   }
   Output Animation Pose(Result=@n1)
  ]
 }
 Transition {
  path="AnimGraph.AnimGraphNode_StateMachine_0.Cane/NoCane States.AnimStateNode_4.Idle_Cane.AnimGraphNode_StateMachine_1.Idle && Reactive (Cane).AnimStateNode_2.Idle.AnimGraphNode_StateMachine_1.Idle (Cane).AnimStateTransitionNode_2.Transition"
  nodes [
   n1 = "Get Relevant Anim Time Remaining (Idle)"!() -> (ReturnValue:float) {is=title K2Node_AnimGetter}
   Result(bCanEnterTransition=AND Boolean(A=Get UseTheForce(), B="float <= float"(A=@n1, B=0.2) -> (ReturnValue=~)))
  ]
 }
 Transition {
  path="AnimGraph.AnimGraphNode_StateMachine_0.Cane/NoCane States.AnimStateNode_4.Idle_Cane.AnimGraphNode_StateMachine_1.Idle && Reactive (Cane).AnimStateNode_2.Idle.AnimGraphNode_StateMachine_1.Idle (Cane).AnimStateTransitionNode_3.Transition"
  nodes [Result()]
 }
 ChugJug {
  path="AnimGraph.AnimGraphNode_StateMachine_0.Cane/NoCane States.AnimStateNode_4.Idle_Cane.AnimGraphNode_StateMachine_1.Idle && Reactive (Cane).AnimStateNode_3.ChugJug"
  nodes [
   n1 = Sequence Player 'WidePlane_Reactive_CaneLaugh_PACK'() {bLoopAnimation=false, sequence=&WidePlane_Reactive_CaneLaugh_PACK}
   Output Animation Pose(Result=@n1)
  ]
 }
 Transition {
  path="AnimGraph.AnimGraphNode_StateMachine_0.Cane/NoCane States.AnimStateNode_4.Idle_Cane.AnimGraphNode_StateMachine_1.Idle && Reactive (Cane).AnimStateTransitionNode_0.Transition"
  nodes [Result(bCanEnterTransition=Get ChugJug Animation())]
 }
 Transition {
  path="AnimGraph.AnimGraphNode_StateMachine_0.Cane/NoCane States.AnimStateNode_4.Idle_Cane.AnimGraphNode_StateMachine_1.Idle && Reactive (Cane).AnimStateTransitionNode_2.Transition"
  nodes [Result(bCanEnterTransition=Get ChestOpen Animation())]
 }
 Transition {
  path="AnimGraph.AnimGraphNode_StateMachine_0.Cane/NoCane States.AnimStateNode_4.Idle_Cane.AnimGraphNode_StateMachine_1.Idle && Reactive (Cane).AnimStateTransitionNode_3.Transition"
  nodes [Result()]
 }
 Transition {
  path="AnimGraph.AnimGraphNode_StateMachine_0.Cane/NoCane States.AnimStateNode_4.Idle_Cane.AnimGraphNode_StateMachine_1.Idle && Reactive (Cane).AnimStateTransitionNode_5.Transition"
  nodes [Result()]
 }
 Transition {
  path=AnimGraph.AnimGraphNode_StateMachine_0.Cane/NoCane States.AnimStateTransitionNode_1.Transition
  nodes [Result(bCanEnterTransition=NOT Boolean(A=Get IsWidePlanePickaxeEquipped() -> (IsWidePlanePickaxeEquipped:bool)))]
 }
 Transition {
  path=AnimGraph.AnimGraphNode_StateMachine_0.Cane/NoCane States.AnimStateTransitionNode_3.Transition
  nodes [Result(bCanEnterTransition=Get IsWidePlanePickaxeEquipped())]
 }
 Idle and Reactive {
  entry=Idle
  path=AnimGraph.AnimGraphNode_StateMachine_1.Idle and Reactive
  ids=id
  nodes [
   AnimStateNode_0 = Idle()
   AnimStateNode_1 = Chest_Open()
   AnimStateNode_2 = Shield() {is=title AnimStateNode}
   AnimStateTransitionNode_0 = AnimStateTransitionNode()
   AnimStateTransitionNode_1 = AnimStateTransitionNode()
   AnimStateTransitionNode_2 = AnimStateTransitionNode()
   AnimStateTransitionNode_3 = AnimStateTransitionNode()
   AnimStateTransitionNode_6 = AnimStateTransitionNode() {bAutomaticRuleBasedOnSequencePlayerInState=true}
   AnimStateTransitionNode_7 = AnimStateTransitionNode() {bAutomaticRuleBasedOnSequencePlayerInState=true}
  ]
  links [
   {
    dst=Idle
    rule=AnimGraph.AnimGraphNode_StateMachine_1.Idle and Reactive.AnimStateTransitionNode_6.Transition
    src=Chest_Open
    via=AnimStateTransitionNode_6
   }
   {
    dst=Shield
    rule=AnimGraph.AnimGraphNode_StateMachine_1.Idle and Reactive.AnimStateTransitionNode_2.Transition
    src=Chest_Open
    via=AnimStateTransitionNode_2
   }
   {
    dst=Chest_Open
    rule=AnimGraph.AnimGraphNode_StateMachine_1.Idle and Reactive.AnimStateTransitionNode_0.Transition
    src=Idle
    via=AnimStateTransitionNode_0
   }
   {
    dst=Shield
    rule=AnimGraph.AnimGraphNode_StateMachine_1.Idle and Reactive.AnimStateTransitionNode_1.Transition
    src=Idle
    via=AnimStateTransitionNode_1
   }
   {
    dst=Chest_Open
    rule=AnimGraph.AnimGraphNode_StateMachine_1.Idle and Reactive.AnimStateTransitionNode_3.Transition
    src=Shield
    via=AnimStateTransitionNode_3
   }
   {
    dst=Idle
    rule=AnimGraph.AnimGraphNode_StateMachine_1.Idle and Reactive.AnimStateTransitionNode_7.Transition
    src=Shield
    via=AnimStateTransitionNode_7
   }
  ]
 }
 Idle {
  path=AnimGraph.AnimGraphNode_StateMachine_1.Idle and Reactive.AnimStateNode_0.Idle
  nodes [
   AnimGraphNode_StateMachine_1 = Idle / Meditate() {is=title AnimGraphNode_StateMachine}
   Output Animation Pose(Result=@AnimGraphNode_StateMachine_1)
  ]
 }
 Idle / Meditate {
  entry=Idle
  path=AnimGraph.AnimGraphNode_StateMachine_1.Idle and Reactive.AnimStateNode_0.Idle.AnimGraphNode_StateMachine_1.Idle / Meditate
  ids=id
  nodes [
   AnimStateNode_0 = Idle()
   AnimStateNode_1 = Meditate() {stateEntered {guid=#15, notifyName=EnteredMeditate}}
   AnimStateTransitionNode_4 = AnimStateTransitionNode()
   AnimStateTransitionNode_5 = AnimStateTransitionNode() {bAutomaticRuleBasedOnSequencePlayerInState=true}
  ]
  links [
   {
    dst=Meditate
    rule=AnimGraph.AnimGraphNode_StateMachine_1.Idle and Reactive.AnimStateNode_0.Idle.AnimGraphNode_StateMachine_1.Idle / Meditate.AnimStateTransitionNode_4.Transition
    src=Idle
    via=AnimStateTransitionNode_4
   }
   {
    dst=Idle
    rule=AnimGraph.AnimGraphNode_StateMachine_1.Idle and Reactive.AnimStateNode_0.Idle.AnimGraphNode_StateMachine_1.Idle / Meditate.AnimStateTransitionNode_5.Transition
    src=Meditate
    via=AnimStateTransitionNode_5
   }
  ]
 }
 Idle {
  path=AnimGraph.AnimGraphNode_StateMachine_1.Idle and Reactive.AnimStateNode_0.Idle.AnimGraphNode_StateMachine_1.Idle / Meditate.AnimStateNode_0.Idle
  nodes [
   n1 = Sequence Player 'WidePlane_Idle_Cane_PACK'() {groupName=NewIdle, method=SyncGroup, sequence=&WidePlane_Idle_Cane_PACK}
   n2 = Sequence Player 'WidePlane_Idle_NoCane_PACK'() {
    groupName=NewIdle
    method=SyncGroup
    sequence=&WidePlane_Idle_NoCane_PACK
   }
   Output Animation Pose(Result=Blend Poses by bool(BlendPose_0=@n2, BlendPose_1=@n1, bActiveValue=Get IsWidePlanePickaxeEquipped()) {blendType=Linear})
  ]
 }
 Meditate {
  path=AnimGraph.AnimGraphNode_StateMachine_1.Idle and Reactive.AnimStateNode_0.Idle.AnimGraphNode_StateMachine_1.Idle / Meditate.AnimStateNode_1.Meditate
  nodes [
   n1 = Sequence Player 'WidePlane_IdleBreak_Meditate_Cane_PACK'() {
    bLoopAnimation=false
    groupName=NewMeditate
    method=SyncGroup
    sequence=&WidePlane_IdleBreak_Meditate_Cane_PACK
   }
   n2 = Sequence Player 'WidePlane_IdleBreak_Meditate_NoCane_PACK'() {
    bLoopAnimation=false
    groupName=NewMeditate
    method=SyncGroup
    sequence=&WidePlane_IdleBreak_Meditate_NoCane_PACK
   }
   Output Animation Pose(Result=Blend Poses by bool(BlendPose_0=@n2, BlendPose_1=@n1, bActiveValue=Get IsWidePlanePickaxeEquipped()) {blendType=Linear})
  ]
 }
 Transition {
  path=AnimGraph.AnimGraphNode_StateMachine_1.Idle and Reactive.AnimStateNode_0.Idle.AnimGraphNode_StateMachine_1.Idle / Meditate.AnimStateTransitionNode_4.Transition
  nodes [
   n1 = "Get Relevant Anim Time Remaining (Idle)"()
   Result(bCanEnterTransition=AND Boolean(A=Get UseTheForce(), B="float <= float"(A=@n1, B=0.2) -> (ReturnValue=~)))
  ]
 }
 Transition {
  path=AnimGraph.AnimGraphNode_StateMachine_1.Idle and Reactive.AnimStateNode_0.Idle.AnimGraphNode_StateMachine_1.Idle / Meditate.AnimStateTransitionNode_5.Transition
  nodes [Result()]
 }
 Chest_Open {
  path=AnimGraph.AnimGraphNode_StateMachine_1.Idle and Reactive.AnimStateNode_1.Chest_Open
  nodes [
   n1 = Sequence Player 'WidePlane_Reactive_CaneCurious_PACK'() {
    bLoopAnimation=false
    groupName=NewCurious
    method=SyncGroup
    sequence=&WidePlane_Reactive_CaneCurious_PACK
   }
   n2 = Sequence Player 'WidePlane_Reactive_NoCaneCurious_PACK'() {
    bLoopAnimation=false
    groupName=NewCurious
    method=SyncGroup
    sequence=&WidePlane_Reactive_NoCaneCurious_PACK
   }
   Output Animation Pose(Result=Blend Poses by bool(BlendPose_0=@n2, BlendPose_1=@n1, bActiveValue=Get IsWidePlanePickaxeEquipped()) {blendType=Linear})
  ]
 }
 Shield {
  path=AnimGraph.AnimGraphNode_StateMachine_1.Idle and Reactive.AnimStateNode_2.Shield
  nodes [
   n1 = Sequence Player 'WidePlane_Reactive_CaneLaugh_PACK'() {
    bLoopAnimation=false
    groupName=NewLaugh
    method=SyncGroup
    sequence=&WidePlane_Reactive_CaneLaugh_PACK
   }
   n2 = Sequence Player 'WidePlane_Reactive_NoCaneLaugh_PACK'() {
    bLoopAnimation=false
    groupName=NewLaugh
    method=SyncGroup
    sequence=&WidePlane_Reactive_NoCaneLaugh_PACK
   }
   Output Animation Pose(Result=Blend Poses by bool(BlendPose_0=@n2, BlendPose_1=@n1, bActiveValue=Get IsWidePlanePickaxeEquipped()) {blendType=Linear})
  ]
 }
 Transition {
  path=AnimGraph.AnimGraphNode_StateMachine_1.Idle and Reactive.AnimStateTransitionNode_0.Transition
  nodes [Result(bCanEnterTransition=Get ChestOpen Animation())]
 }
 Transition {
  path=AnimGraph.AnimGraphNode_StateMachine_1.Idle and Reactive.AnimStateTransitionNode_1.Transition
  nodes [Result(bCanEnterTransition=Get ChugJug Animation())]
 }
 Transition {
  path=AnimGraph.AnimGraphNode_StateMachine_1.Idle and Reactive.AnimStateTransitionNode_2.Transition
  nodes [Result(bCanEnterTransition=Get ChugJug Animation())]
 }
 Transition {
  path=AnimGraph.AnimGraphNode_StateMachine_1.Idle and Reactive.AnimStateTransitionNode_3.Transition
  nodes [Result(bCanEnterTransition=Get ChestOpen Animation())]
 }
 Transition {path=AnimGraph.AnimGraphNode_StateMachine_1.Idle and Reactive.AnimStateTransitionNode_6.Transition, nodes [Result()]}
 Transition {path=AnimGraph.AnimGraphNode_StateMachine_1.Idle and Reactive.AnimStateTransitionNode_7.Transition, nodes [Result()]}
 EventGraph {nodes [
   n1 = Event BlueprintUpdateAnimation()
   > n2 = Sequence!(execute) -> (then_0, then_1, then_2:Exec)
   > n3 = Gravity Override State Function DEPRECATED(CharacterPartAnimInstance=Self-Reference(), GameTimeInSeconds=@n1.DeltaTimeX, GravityOverrideParams=Get Gravity Override Params())
   > Set Gravity Override(Gravity Override=@n3)
   n4 = Branch(execute=@n2.then_2, Condition=Get bIsSkyDiving())
   > Set IsWidePlanePickaxeEquipped(IsWidePlanePickaxeEquipped:bool=false) -> (Output_Get:bool)
   n5 = GetCurrentWeapon!(self:Fort Pawn Object Reference=Get As Fort Player Pawn() -> (As Fort Player Pawn:Fort Player Pawn Object Reference)) -> (ReturnValue:Fort Weapon Object Reference)
   Set IsWidePlanePickaxeEquipped(
    execute=@n4.else
    IsWidePlanePickaxeEquipped=Contains(SearchIn="To String (Object)"!(InObj:Object Reference=@n5) -> (ReturnValue:String), Substring=B_Athena_Pickaxe_WidePlane)
   )
   n6 = Branch(execute=@n2.then_1, Condition=Get bIsSkyDiving())
   > Set PBA_Alpha(PBA_Alpha:double=0.1) -> (Output_Get:double)
   Set PBA_Alpha(execute=@n6.else, PBA_Alpha=0.35)
   AnimNotify_GoIntoMeditate() {is=title K2Node_Event}
   > Branch(Condition=RandomBoolWithWeight!(Weight:float=0.3) -> (ReturnValue:bool))
   > Set UseTheForce(UseTheForce:bool=true) -> (Output_Get:bool)
   n7 = Event BlueprintProcessAnimInputEvent()
   > Branch(Condition=IsNamed(self=@n7.AnimInputEvent, InEventName=aie.ChestOpened))
   > Set ChestOpen Animation(ChestOpen Animation:bool=true) -> (Output_Get:bool)
   Event BlueprintPostEvaluateAnimation() {is=title K2Node_Event}
   > Set ChugJug Animation(ChugJug Animation:bool=false) -> (Output_Get:bool)
   > Set ChestOpen Animation(ChestOpen Animation=false)
   Event BlueprintInitializeAnimation()
   > n8 = Sequence!(execute) -> (then_0, then_1)
   > n9 = Cast To FortPlayerPawn(Object=TryGetPawnOwner())
   > n10 = Set As Fort Player Pawn(As Fort Player Pawn:Fort Player Pawn Object Reference=@n9) -> (Output_Get:Fort Player Pawn Object Reference)
   > n11 = WatchCosmeticStat!(
    execute:Exec
    Pawn:Fort Player Pawn Object Reference=@n10
    StatTag:Fort Stat Manager Tag Structure=(TagName="StatManager.GainedShieldTimes")
    DefaultValue:int=0
    bBroadcastInitialValue:bool=false
    InitialValueBroadcastDelay:float=-1
    ChangeValueBroadcastDelay:float=-1
    bConsumeWaitingNotifications:bool=true
   ) -> (then:Exec, OnStatChanged:Exec, StatName:Name, StatValue:int) {is=title K2Node_AsyncAction}
   Set ChugJug Animation(execute=@n11.OnStatChanged, ChugJug Animation=true)
   Branch(execute=[@n8.then_1, @n14.then], Condition=Get bIsInFrontEnd())
   > Delay(Duration=13)
   > Set ChugJug Animation(ChugJug Animation=true)
   AnimNotify_EnteredMeditate() {is=title K2Node_Event}
   > Set UseTheForce(UseTheForce=false)
   AnimNotify_PlayReactiveAudio() {is=title K2Node_Event}
   > n12 = Branch(Condition=Get bIsInFrontEnd())
   n13 = Get!(execute:Exec=@n12.else) -> (then:Exec, else:Exec, As Fort Player Pawn:Fort Player Pawn Object Reference)
   > SpawnSoundAttached!(
    execute
    Sound:Sound Base Object Reference=Get ReactiveSoundCue() -> (ReactiveSoundCue:Sound Base Object Reference)
    AttachToComponent:Scene Component Object Reference=Get RootComponent!(self:Actor Object Reference=@n13) -> (RootComponent:Scene Component Object Reference)
    AttachPointName:Name=None
    Location:Vector
    Rotation:Rotator
    LocationType:EAttachLocation Enum=KeepRelativeOffset
    bStopWhenAttachedToDestroyed:bool=false
    VolumeMultiplier:float=1
    PitchMultiplier:float=1
    StartTime:float=0
    AttenuationSettings:Sound Attenuation Object Reference
    ConcurrencySettings:Sound Concurrency Object Reference
    bAutoDestroy:bool=true
   ) -> (then, ReturnValue:Audio Component Object Reference)
   AnimNotify_FE_ReactiveReset() {is=title K2Node_Event}
   > n14 = Set ChugJug Animation(ChugJug Animation=false)
   n15 = OR Boolean(A=Get ChestOpen Animation(), B=Get ChugJug Animation())
   n16 = Branch(Condition=@n15)
   > n17 = Set PBA_Alpha(PBA_Alpha=0)
   Branch(execute=[@n17.then, @n16.else], Condition=NOT Boolean(A=@n15))
   > Set PBA_Alpha(PBA_Alpha=1)
   Branch(Condition=Get bIsSkyDiving())
   > Set PBA_Alpha(PBA_Alpha=0.6)
   n18 = GetOwnerSkeletalMeshComponent()
   n19 = BreakRotator(InRot=BreakTransform(InTransform=GetDeltaTransformFromRefPose(self=@n18, BoneName=spine_05, BaseName=spine_01)).Rotation)
   Set backpack_rotate(backpack_rotate=MakeRotator(Roll=MapRangeClamped(Value=@n19.Yaw, InRangeA=0, InRangeB=-30, OutRangeB=35)))
   > Set backpack_translation(backpack_translation=MakeVector(Y=MapRangeClamped(Value=@n19.Yaw, InRangeA=0, InRangeB=-30, OutRangeB=-10)))
   TryGetPawnOwner()
  ]}
]
variables [
 As Fort Player Pawn {type=FortPlayerPawn}
 As Player Pawn Athena {type=PlayerPawn_Athena_C}
 backpack_rotate {type=Rotator}
 backpack_translation {type=Vector}
 ChestOpen Animation {type=boolean}
 ChugJug Animation {type=boolean}
 Control And Modifier Parameters {
  type=PhysicsControlControlAndModifierParameters
  default {
   controlMultiplierParameters []
   controlParameters [
    Arm_R {data {
      angularDampingRatio=40
      angularExtraDamping=0
      angularStrength=1
      angularTargetVelocityMultiplier=1
      bDisableCollision=false
      bEnableAngularDampingRatio=true
      bEnableAngularExtraDamping=true
      bEnableAngularStrength=true
      bEnableAngularTargetVelocityMultiplier=true
      bEnableCustomControlPoint=true
      bEnableLinearDampingRatio=true
      bEnableLinearExtraDamping=true
      bEnableLinearStrength=true
      bEnableLinearTargetVelocityMultiplier=true
      bEnableMaxForce=true
      bEnableMaxTorque=true
      bEnablebDisableCollision=true
      bEnablebEnabled=true
      bEnablebOnlyControlChildObject=true
      bEnablebUseAccelerationDriveMode=true
      bEnablebUseCustomControlPoint=true
      bEnablebUseSkeletalAnimation=true
      bEnabled=true
      bOnlyControlChildObject=false
      bUseAccelerationDriveMode=true
      bUseCustomControlPoint=false
      bUseSkeletalAnimation=true
      linearDampingRatio=1
      linearExtraDamping=0
      linearStrength=100
      linearTargetVelocityMultiplier=1
      maxForce=0
      maxTorque=0
      customControlPoint=<0 0 0>
     }}
    Head {data {
      angularDampingRatio=10
      angularExtraDamping=0
      angularStrength=1
      angularTargetVelocityMultiplier=1
      bDisableCollision=false
      bEnableAngularDampingRatio=true
      bEnableAngularExtraDamping=true
      bEnableAngularStrength=true
      bEnableAngularTargetVelocityMultiplier=true
      bEnableCustomControlPoint=true
      bEnableLinearDampingRatio=true
      bEnableLinearExtraDamping=true
      bEnableLinearStrength=true
      bEnableLinearTargetVelocityMultiplier=true
      bEnableMaxForce=true
      bEnableMaxTorque=true
      bEnablebDisableCollision=true
      bEnablebEnabled=true
      bEnablebOnlyControlChildObject=true
      bEnablebUseAccelerationDriveMode=true
      bEnablebUseCustomControlPoint=true
      bEnablebUseSkeletalAnimation=true
      bEnabled=true
      bOnlyControlChildObject=false
      bUseAccelerationDriveMode=true
      bUseCustomControlPoint=false
      bUseSkeletalAnimation=true
      linearDampingRatio=0
      linearExtraDamping=0
      linearStrength=0
      linearTargetVelocityMultiplier=1
      maxForce=0
      maxTorque=0
      customControlPoint=<0 0 0>
     }}
   ]
   modifierParameters [Arm_L {data {
      bEnableCCD=false
      bEnableCollisionType=true
      bEnableGravityMultiplier=true
      bEnableKinematicTargetSpace=true
      bEnableMovementType=true
      bEnablePhysicsBlendWeight=true
      bEnablebEnableCCD=true
      bEnablebUpdateKinematicFromSimulation=true
      bUpdateKinematicFromSimulation=true
      collisionType=QueryAndPhysics
      gravityMultiplier=1
      kinematicTargetSpace=OffsetInBoneSpace
      movementType=Kinematic
      physicsBlendWeight=1
     }}]
  }
 }
 Control And Modifier Updates {
  type=PhysicsControlControlAndModifierUpdates
  default {
   controlMultiplierUpdates []
   controlUpdates []
   modifierUpdates [
    Arm_L {data {
      bEnableCCD=false
      bEnableCollisionType=true
      bEnableGravityMultiplier=true
      bEnableKinematicTargetSpace=true
      bEnableMovementType=true
      bEnablePhysicsBlendWeight=true
      bEnablebEnableCCD=true
      bEnablebUpdateKinematicFromSimulation=true
      bUpdateKinematicFromSimulation=true
      collisionType=QueryAndPhysics
      gravityMultiplier=1
      kinematicTargetSpace=OffsetInBoneSpace
      movementType=Simulated
      physicsBlendWeight=1
     }}
    None {data {
      bEnableCCD=false
      bEnableCollisionType=true
      bEnableGravityMultiplier=true
      bEnableKinematicTargetSpace=true
      bEnableMovementType=true
      bEnablePhysicsBlendWeight=true
      bEnablebEnableCCD=true
      bEnablebUpdateKinematicFromSimulation=true
      bUpdateKinematicFromSimulation=true
      collisionType=QueryAndPhysics
      gravityMultiplier=1
      kinematicTargetSpace=OffsetInBoneSpace
      movementType=Simulated
      physicsBlendWeight=1
     }}
   ]
  }
 }
 Gravity Override {type=Vector}
 Gravity Override Params {
  type=GravityOverrideParamsStruct
  default {
   skydive_DivingDown {gravityOverride=<0 0 0>}
   skydive_FloatingDown {gravityOverride=<0 0 0>}
   skydive_GoingUp {gravityOverride=<0 0 0>}
   skydive_Parachuting {gravityOverride=<0 0 0>}
  }
 }
 IsWidePlanePickaxeEquipped {type=boolean}
 PBA_Alpha {type=number, default=1}
 ReactiveSoundCue {type=SoundBase, default=/BRCosmetics/Athena/Sounds/BackBlings/WidePlane/IP_Backbling_WidePlane_SlurpReact_Cue}
 Sim Space Settings {
  type=SimSpaceSettings
  default {
   dampingAlpha=1
   maxAngularAcceleration=100
   maxAngularVelocity=4
   maxLinearAcceleration=300
   maxLinearVelocity=500
   velocityScaleZ=1
   worldAlpha=0.6
   externalLinearDragV=<0 0 0>
  }
 }
 Timer {type=number}
 UseTheForce {type=boolean}
]
